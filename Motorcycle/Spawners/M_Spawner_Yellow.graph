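# M_Spawner_Yellow is a flat yellow-orange Default Lit material, apparently for motorcycle spawners, going by its name and folder.
# It holds only a constant color and a constant scalar, with no textures or parameters; the file does not show which inputs they feed.
# Usage flags enable it on skeletal meshes and with mesh deformers.
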Material /Motorcycle/Spawners/M_Spawner_Yellow
graph {
 properties {bUsedWithMeshDeformer=true, bUsedWithSkeletalMesh=true, pixelDepthOffsetMode=PDOM_Legacy, usedShadingModels=Default Lit}
 nodes [
  Constant3Vector() {constant=<rgba 0.9722 0.493 0 0>}
  Constant() {r=0.7}
 ]
}
